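# Niagara system for the glitch reaction effect on the TerraBulb backpack cosmetic.
# Two emitters, Box_rate and Glitch_sprite, continuously spawn short-lived mesh particles with randomized non-uniform scale.
# Each update re-places particles in a small box around a socket read through the system Socket Reader.
# Curves animate size, alpha fade and rapidly shifting HDR colors, with camera-distance scaling on Box_rate size and a front-end spawn boost on Glitch_sprite.
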NiagaraSystem /BRCosmetics/Accessories/FORT_Backpacks/Backpack_F_MED_TerraBulb/FX/NS_Backpack_TeraBulb_ReactFX_Glitch
dependencies {
 usedDataInterfaces [&NiagaraDataInterfaceCurve, &NiagaraDataInterfaceColorCurve, &NiagaraDataInterfaceSocketReader]
 usedDynamicInputs [&FNTimeOfDayColor]
 usedModules [
  &self:SystemSpawnScript.SystemScriptSource.SystemScriptGraph.NiagaraNodeAssignment_1.SetVariables_#1
  &self:SystemSpawnScript.SystemScriptSource.SystemScriptGraph.NiagaraNodeAssignment_0.SetVariables_#2
  &SystemState
 ]
}
emitters {
 Box_rate {
  emitterUpdateScript [
   &EmitterState()
   &SpawnRate(SpawnRate=&Multiply_Float(A=150, B=1), Use Spawn Probability=false)
  ]
  particleSpawnScript [
   &InitializeParticle(
    Lifetime Max=0.15
    Lifetime Min=0.1
    Lifetime Mode=::NewEnumerator1
    Mesh Scale Max=<3 3 10>
    Mesh Scale Min=<0.5 0.5 5>
    Mesh Scale Mode=::NewEnumerator2
    Position Offset=<0 0 100>
    UsePositionOffset=true
    Write Material Random=true
   )
   &CameraOffset(Camera Offset Amount=1)
  ]
  particleUpdateScript [
   &ParticleState()
   &SocketLocation(
    SamplePosOffset=false
    Sampled Orientation Coordinate Space:ENiagaraCoordinateSpace=::Local
    Socket Reader=$System.Socket Reader
    Socket Sampling Mode=::NewEnumerator3
    VelocitySpeedLimit=false
   )
   &ShapeLocation(
    Box Size=<5 5 8>
    MaskSpawnGroup=false
    Offset=<0 0 8>
    Offset Mode=::NewEnumerator0
    Override Seed=false
    Shape Origin=$Output.SocketLocation.SampledPosition
    Shape Primitive=::NewEnumerator2
    Surface Only=false
    UseOwnerScale=true
   )
   &ScaleMeshSize(Scale Factor=&Multiply_VectorByFloat(
     Vector=<1 1 1>
     Float=&Multiply_Float(
      A=0.07
      B=&FloatFromCurve(
       FloatCurve=DataInterface() {propertyValues {Curve {keys [
           {arriveTangent=4.1785, time=0, value=0}
           <at 0.5 1.1>
           {arriveTangent=-0.0405, interpMode=RCIM_Cubic, leaveTangent=-0.0405, tangentMode=RCTM_User, time=1, value=1.5}
          ]}}}
       Scale Curve=&Multiply_Float(
        A=1
        B=&ClampFloat(
         Float=&FN_LOD_Distance_Remap(Start Distance=200, End Distance=400, Minimum Output=0.5, Maximum Output=1)
         Min=0.2
         Max=0.5
        )
       )
      )
     )
    ))
   &Drag(Drag=4, Use Linear Drag=true, Use Rotational Drag=true)
   &SolveForcesAndVelocity(Clamp Velocity=false, Limit Acceleration=false)
   &ScaleColor(
    Scale Alpha=&Multiply_Float(
     A=&FloatFromCurve(
      FloatCurve=DataInterface() {propertyValues {Curve {keys [<at 0 1>,<at 1 0>]}}}
      Scale Curve=&RandomRangeFloat(Minimum=0.4, Maximum=1)
     )
     B=3
    )
    Scale RGB=&Multiply_VectorByFloat(
     Vector=&MakeVectorFromLinearColorRGB(Color=&ColorFromCurve(ColorCurve=DataInterface() {propertyValues {
         AlphaCurve {keys []}
         BlueCurve {keys [
           <at 0 5 RCIM_Cubic>
           <at 0.1326 5>
           <at 0.2869 0>
           <at 0.4134 0>
           <at 0.5108 5>
           <at 0.6236 0.0803>
           {arriveTangent=26.4359, interpMode=RCIM_Cubic, leaveTangent=26.4359, time=0.7107, value=0}
           <at 0.8097 5>
          ]}
         GreenCurve {keys [
           <at 0 4.702 RCIM_Cubic>
           <at 0.1326 0>
           <at 0.2869 2.0314>
           <at 0.4134 0.0442>
           <at 0.5108 0.7585>
           <at 0.6236 5>
           {arriveTangent=-21.532, interpMode=RCIM_Cubic, leaveTangent=-21.532, time=0.7107, value=0.3634}
           <at 0.8097 0.9929>
          ]}
         RedCurve {keys [<at 0 0 RCIM_Cubic>,<at 0.1326 2.465>,<at 0.2869 5>,<at 0.4134 5>,<at 0.5108 4.6237>,<at 0.6236 0>,<at 0.7107 1 RCIM_Cubic>,<at 0.8097 0>]}
        }}))
     Float=&FloatFromCurve(
      FloatCurve=DataInterface() {propertyValues {Curve {keys [{arriveTangent=-2.0684, interpMode=RCIM_Cubic, leaveTangent=-2.0684, tangentMode=RCTM_User, time=0, value=2}, <at 1 0>]}}}
      Scale Curve=&Multiply_Float(A=&Multiply_Float(A=3, B=1), B=1)
     )
    )
    ScaleA=true
    ScaleRGB=true
   )
  ]
  rendererClasses [&NiagaraMeshRendererProperties]
  renderers [{rendererClass=&NiagaraMeshRendererProperties, rendererIndex=0}]
 }
 Glitch_sprite {
  emitterUpdateScript [
   &EmitterState()
   &SpawnRate(
    SpawnRate=&Multiply_Float(A=70, B=&FN_FrontEndFloatModifier(Gameplay=1, Front End=2))
    Use Spawn Probability=false
   )
  ]
  particleSpawnScript [&InitializeParticle(
    Color=<rgba 0 5 1.3182 2>
    Lifetime Max=0.3
    Lifetime Min=0.3
    Lifetime Mode=::NewEnumerator1
    Mesh Scale Max=&MakeCustomVectorFromBool(Bool=&RandomBool(), True Vector=<2 5 1>, False Vector=<2 2 6.5>)
    Mesh Scale Min=<1 1 3>
    Mesh Scale Mode=::NewEnumerator2
    Position Offset=<0 0 100>
    UsePositionOffset=true
    Write Material Random=true
   )]
  particleUpdateScript [
   &ParticleState()
   &SocketLocation(
    SamplePosOffset=false
    Sampled Orientation Coordinate Space:ENiagaraCoordinateSpace=::Local
    Socket Reader=$System.Socket Reader
    Socket Sampling Mode=::NewEnumerator3
    VelocitySpeedLimit=false
   )
   &ShapeLocation(
    Box Size=<5 5 8>
    MaskSpawnGroup=false
    Offset=<0 -1.3 8>
    Offset Coordinate Space=::Local
    Offset Mode=::NewEnumerator0
    Override Seed=false
    Shape Origin=$Output.SocketLocation.SampledPosition
    Shape Primitive=::NewEnumerator2
    Surface Only=false
    UseOwnerScale=true
   )
   &ScaleMeshSize(Scale Factor=&Multiply_VectorByFloat(
     Vector=<1 1 1>
     Float=&Multiply_Float(
      A=0.2
      B=&FloatFromCurve(
       FloatCurve=DataInterface() {propertyValues {Curve {keys [
           {arriveTangent=4.1785, time=0, value=1}
           <at 0.5 1.1>
           {arriveTangent=-0.0405, interpMode=RCIM_Cubic, leaveTangent=-0.0405, tangentMode=RCTM_User, time=1, value=1.5}
          ]}}}
       Scale Curve=0.35
      )
     )
    ))
   &Drag(Drag=2, Use Linear Drag=true, Use Rotational Drag=true)
   &SolveForcesAndVelocity(Clamp Velocity=false, Limit Acceleration=false)
   &DynamicMaterialParameters(Module.Param0WriteEnabled=true, Module.Param1WriteEnabled=true, Module.Param2WriteEnabled=true, Module.Param3WriteEnabled=true) {enabled=false}
   &ScaleColor(
    Scale Alpha=&Multiply_Float(
     A=&FloatFromCurve(
      FloatCurve=DataInterface() {propertyValues {Curve {keys [
          {arriveTangent=9.0725, interpMode=RCIM_Cubic, leaveTangent=9.0725, tangentMode=RCTM_User, time=0, value=0}
          {arriveTangent=-2.5249, interpMode=RCIM_Cubic, leaveTangent=-2.5249, tangentMode=RCTM_User, time=0.2, value=0.9847}
          {arriveTangent=-0.0405, interpMode=RCIM_Cubic, leaveTangent=-0.0405, tangentMode=RCTM_User, time=1, value=0}
         ]}}}
      Scale Curve=&RandomRangeFloat(Minimum=0.2, Maximum=1)
     )
     B=1
    )
    Scale RGB=&Multiply_VectorByFloat(
     Vector=&MakeVectorFromLinearColorRGB(Color=&ColorFromCurve(ColorCurve=DataInterface() {propertyValues {
         AlphaCurve {keys []}
         BlueCurve {keys [
           <at 0 5 RCIM_Cubic>
           <at 0.1326 5>
           <at 0.2869 5>
           <at 0.4134 0>
           <at 0.5108 5>
           <at 0.6236 0.0803>
           {arriveTangent=26.436, interpMode=RCIM_Cubic, leaveTangent=26.436, time=0.7107, value=0.8321}
           <at 0.8097 5>
          ]}
         GreenCurve {keys [
           <at 0 4.702 RCIM_Cubic>
           <at 0.1326 0>
           <at 0.2869 0.6611>
           <at 0.4134 0.0442>
           <at 0.5108 0.7585>
           <at 0.6236 5>
           {arriveTangent=-21.532, interpMode=RCIM_Cubic, leaveTangent=-21.532, time=0.7107, value=1}
           <at 0.8097 0.9929>
          ]}
         RedCurve {keys [<at 0 0 RCIM_Cubic>,<at 0.1326 2.465>,<at 0.2869 0>,<at 0.4134 5>,<at 0.5108 4.6237>,<at 0.6236 0>,<at 0.7107 0 RCIM_Cubic>,<at 0.8097 0>]}
        }}))
     Float=&FloatFromCurve(
      FloatCurve=DataInterface() {propertyValues {Curve {keys [{arriveTangent=-2.0684, interpMode=RCIM_Cubic, leaveTangent=-2.0684, tangentMode=RCTM_User, time=0, value=2}, <at 1 0>]}}}
      Scale Curve=&Multiply_Float(A=1, B=&FN_FrontEndFloatModifier(Gameplay=1, Front End=1))
     )
    )
    ScaleA=true
    ScaleRGB=true
   )
  ]
  rendererClasses [&NiagaraMeshRendererProperties]
  renderers [{rendererClass=&NiagaraMeshRendererProperties, rendererIndex=0}]
 }
}
summary {
 systemName=NS_Backpack_TeraBulb_ReactFX_Glitch
 emitters [
  {emitterName=Box_rate, rendererClasses [&NiagaraMeshRendererProperties]}
  {emitterName=Glitch_sprite, rendererClasses [&NiagaraMeshRendererProperties]}
 ]
 userVariables [
  User.bIsFrontEnd {type=NiagaraBool}
  User.bIsFrontEndPreview {type=NiagaraBool}
 ]
}
