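# Masked character material for the TerraBulb backpack, layering backpack-specific effects between the 2023 CharacterParent PreFX and PostFX functions.
# A tech-art mask's red channel darkens base color and swaps emissive to a flat tint, while green marks the flower for noisy Fresnel and scanline glow.
# FlowerReactiveAlpha blends in a reactive opacity mask and vertex offset, with glitched digital-noise and animated sine-wave networks built from pre-skinned position.
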Material /BRCosmetics/Accessories/FORT_Backpacks/Backpack_F_MED_TerraBulb/Materials/M_TerraBulb_BackpackOverride
graph {
 properties {
  bAutomaticallySetUsageInEditor=false
  bUseMaterialAttributes=true
  bUsedWithMeshDeformer=true
  bUsedWithMorphTargets=true
  bUsedWithSkeletalMesh=true
  blendMode=BLEND_Masked
  pixelDepthOffsetMode=PDOM_Legacy
  shadingModel=MSM_FromMaterialExpression
  usedShadingModels=Default Lit | Subsurface | Clear Coat | Subsurface Profile | Cloth
 }
 nodes [
  n1 = MaterialFunctionCall() {materialFunction=&MF_CharacterParent_2023_PreFX}
  n2 = GetMaterialAttributes(None=@n1.Result) {attributeGetTypes [#1, #2]}
  n3 = NamedRerouteUsage()
  n4 = ScalarParameter() {defaultValue=0.5, group=BackOverride, parameterName=BaseColorDown}
  n5 = TextureSampleParameter2D() {
   group=BackOverride
   parameterName=BackpackOverride
   samplerType=SAMPLERTYPE_Masks
   texture=/BRCosmetics/Accessories/FORT_Backpacks/Backpack_F_MED_TerraBulb/Textures/T_TerraBulbBackpack_TechArtMask
  }
  n6 = VectorParameter() {parameterName=BackGroundEM, defaultValue=<rgba 0 0.0573 0.0316 1>}
  n7 = NamedRerouteDeclaration(None=@n5.G) {name=Flower, nodeColor=<rgba 0.7344 1 0 1>}
  n8 = TextureSample(UVs=Multiply(A=TextureCoordinate(), B=ScalarParameter() {defaultValue=4, group=BackOverride, parameterName=WhiteNoiseTiling})) {texture=/BRCosmetics/Accessories/FORT_Backpacks/Backpack_F_MED_TerraBulb/Textures/T_WhiteNoise}
  n9 = VectorParameter() {group=Cube_Additional, parameterName=EdgeColor, defaultValue=<rgba 0.0416 0.224 0.1529 10>}
  n10 = Multiply(A=@n9.RGB, B=@n9.RGB)
  n11 = Fresnel() {baseReflectFraction=0, exponent=1}
  n12 = LinearInterpolate(Alpha=@n11) {constA=0.5}
  n13 = VectorParameter() {group=Cube_Additional, parameterName=ScanLineColor, defaultValue=<rgba 0 0.224 0.1367 1>}
  n14 = Multiply(A=@n8.R, B=Multiply(A=@n13.RGB, B=@n13.RGB))
  n15 = Add(
   A=Multiply(A=Multiply(A=@n8.R, B=@n10), B=@n11)
   B=Multiply(
    A=Multiply(A=@n12, B=@n14)
    B=MaterialFunctionCall!(PreskinnedPOS, Scanline Amount) {materialFunction=/BRCosmetics/Accessories/FORT_Backpacks/Backpack_M_MED_SpacePlunge/Material/MF_Scanlines}.Result
   )
  )
  n16 = NamedRerouteUsage()
  n17 = NamedRerouteUsage()
  n18 = ScalarParameter() {group=Reactivity, parameterName=FlowerReactiveAlpha}
  n19 = NamedRerouteUsage()
  n20 = NamedRerouteUsage()
  n21 = SetMaterialAttributes!(
   MaterialAttributes=@n2.MaterialAttributes
   Base Color=LinearInterpolate(A=@n3, B=Multiply(A=@n3, B=@n4) {constB=0.1}, Alpha=@n5.R)
   Emissive Color=Add(A=LinearInterpolate(A=@n2.Emissive Color, B=@n6.RGB, Alpha=@n5.R) {constB=0}, B=Multiply(A=@n7, B=@n15))
   Opacity Mask=LinearInterpolate(B=LinearInterpolate(B=@n16, Alpha=@n17) {constA=1}, Alpha=@n18) {constA=1}
   World Position Offset=Multiply(A=Multiply(A=@n19, B=@n20), B=@n18)
  ) {attributeSetTypes [
    #1
    #2
    #3
    #4
   ]}
  MaterialFunctionCall(Material=@n21, M=@n1.M, Pre Skinned Local Position=@n1.M, Distance Blend=@n1.M, Fresnel=@n1.M) {materialFunction=&MF_CharacterParent_2023_PostFX}
  n22 = ComponentMask(None=Subtract(A=NamedRerouteUsage(), B=VectorParameter() {group=Reactivity, parameterName=ReactPOS, defaultValue=<rgba 0 0 10 1>}.RGB)) {b=true, r=true}
  n23 = Divide(A=@n22, B=ScalarParameter() {defaultValue=20, group=Reactivity, parameterName=Unit_React})
  n24 = ComponentMask(None=@n23) {g=true}
  n25 = ScalarParameter() {defaultValue=20, group=Reactivity, parameterName=Skin_React}
  n26 = Add(A=Floor(None=Multiply(A=@n24, B=@n25)), B=Multiply(A=Time(), B=ScalarParameter() {defaultValue=3, parameterName=Hash_React_Speed}))
  n27 = TextureSample(UVs=Add(
    A=MaterialFunctionCall!(UV=@n23) {materialFunction=/Game/Global/Materials/MaterialGym/Materials/HolographicTexture/MF_HolographicTextureGlitch_Grid}.Glitched UV
    B=MaterialFunctionCall(Seed=@n26) {materialFunction=&MF_Hash}.Vec2
   )) {samplerType=SAMPLERTYPE_LinearColor, texture=/Game/Global/Textures/Default/Noise/T_Atlas_DigitalNoise_Small}
  MaterialFunctionCall(A=@n27.R, B=@n27.R, C=@n27.R, Alpha=Time()) {materialFunction=&Lerp_3Color}
  NamedRerouteDeclaration(None=@n2.Base Color) {name=BaseColorIN, nodeColor=<rgba 0.7812 1 0 1>}
  NamedRerouteDeclaration(None=@n1.Pre skinned Local Position) {name=PreskinnedPOS, nodeColor=<rgba 1 0.9609 0 1>}
  NamedRerouteDeclaration(None=@n27.R) {name=Opacity_React, nodeColor=<rgba 0 0.6484 1 1>}
  n28 = ComponentMask(None=Divide(A=@n22, B=ScalarParameter() {defaultValue=40, group=Reactivity, parameterName=WPO_React})) {g=true}
  NamedRerouteDeclaration(None=OneMinus(None=Frac(None=@n28))) {name=Gradient, nodeColor=<rgba 0.3828 0 1 1>}
  n29 = ScalarParameter() {defaultValue=20, group=Reactivity, parameterName=sin_React}
  n30 = Time()
  n31 = ScalarParameter() {defaultValue=-2, parameterName=Sine_React_Speed}
  n32 = Sine(None=Add(A=Multiply(A=@n28, B=@n29), B=Multiply(A=@n30, B=@n31)))
  n33 = Subtract(A=SmoothStep(Value=ConstantBiasScale(None=@n32)) {constMax=0.2}) {constB=0.5}
  n34 = NamedRerouteUsage()
  n35 = Transform(None=Constant3Vector() {constant=<rgba 1 0 0 0>}) {transformSourceType=TRANSFORMSOURCE_Local}
  n36 = ScalarParameter() {defaultValue=3, parameterName=Sine_Intensity}
  NamedRerouteDeclaration(None=Multiply(A=Multiply(A=@n33, B=@n34), B=Multiply(A=@n35, B=@n36))) {name=Sine_WPO, nodeColor=<rgba 0 1 0.0625 1>}
 ]
}
