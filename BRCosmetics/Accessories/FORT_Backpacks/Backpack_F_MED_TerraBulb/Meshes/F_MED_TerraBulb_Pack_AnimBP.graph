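# Animation Blueprint for the TerraBulb backpack, playing elimination and damage reactions over a base pose.
# A front-end flag selects between a timer-driven preview state machine, which periodically replays the elimination reaction, and the gameplay state machine.
# Anim input events briefly pulse booleans via retriggerable delays to trigger reactions, which return to idle when their sequence ends.
# A chest-opened flag is also set, but no transition reads it.
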AnimBlueprint /BRCosmetics/Accessories/FORT_Backpacks/Backpack_F_MED_TerraBulb/Meshes/F_MED_TerraBulb_Pack_AnimBP
target_skeleton=&/F_MED_TerraBulb_Pack_Skeleton
graphs [
 AnimGraph {nodes [
   AnimGraphNode_StateMachine_0 = Idle and Reactive() {is=title AnimGraphNode_StateMachine}
   AnimGraphNode_StateMachine_1 = FE_Reactivity() {is=title AnimGraphNode_StateMachine}
   Output Pose(Result=Blend Poses by bool(BlendPose_0=@AnimGraphNode_StateMachine_1, BlendPose_1=@AnimGraphNode_StateMachine_0, bActiveValue=Get bIsInFrontEnd()) {blendType=Linear}) {layerGroup=None}
  ]}
 Idle and Reactive {
  entry=Idle
  path=AnimGraph.AnimGraphNode_StateMachine_0.Idle and Reactive
  ids=id
  nodes [
   AnimStateNode_1 = Reactive_Elimination() {
    is=title AnimStateNode
    stateEntered {guid=#1}
    stateFullyBlended {guid=#2}
    stateLeft {guid=#3}
   }
   AnimStateNode_2 = Idle() {
    stateEntered {guid=#4}
    stateFullyBlended {guid=#5}
    stateLeft {guid=#6}
   }
   AnimStateNode_3 = Reactive_TakeDamage() {
    is=title AnimStateNode
    stateEntered {guid=#7}
    stateFullyBlended {guid=#8}
    stateLeft {guid=#9}
   }
   AnimStateTransitionNode_0 = AnimStateTransitionNode() {bAllowInertializationForSelfTransitions=true, crossfadeDuration=0}
   AnimStateTransitionNode_1 = AnimStateTransitionNode() {bAllowInertializationForSelfTransitions=true, bAutomaticRuleBasedOnSequencePlayerInState=true, crossfadeDuration=0}
   AnimStateTransitionNode_2 = AnimStateTransitionNode() {
    crossfadeDuration=0
    transitionEnd {guid=#10}
    transitionInterrupt {guid=#11}
    transitionStart {guid=#12}
   }
   AnimStateTransitionNode_3 = AnimStateTransitionNode() {
    bAutomaticRuleBasedOnSequencePlayerInState=true
    crossfadeDuration=0
    transitionEnd {guid=#13}
    transitionInterrupt {guid=#14}
    transitionStart {guid=#15}
   }
  ]
  links [
   {
    dst=Reactive_Elimination
    rule=AnimGraph.AnimGraphNode_StateMachine_0.Idle and Reactive.AnimStateTransitionNode_2.Transition
    src=Idle
    via=AnimStateTransitionNode_2
   }
   {
    dst=Reactive_TakeDamage
    rule=AnimGraph.AnimGraphNode_StateMachine_0.Idle and Reactive.AnimStateTransitionNode_0.Transition
    src=Idle
    via=AnimStateTransitionNode_0
   }
   {
    dst=Idle
    rule=AnimGraph.AnimGraphNode_StateMachine_0.Idle and Reactive.AnimStateTransitionNode_3.Transition
    src=Reactive_Elimination
    via=AnimStateTransitionNode_3
   }
   {
    dst=Idle
    rule=AnimGraph.AnimGraphNode_StateMachine_0.Idle and Reactive.AnimStateTransitionNode_1.Transition
    src=Reactive_TakeDamage
    via=AnimStateTransitionNode_1
   }
  ]
 }
 Reactive_Elimination {
  path=AnimGraph.AnimGraphNode_StateMachine_0.Idle and Reactive.AnimStateNode_1.Reactive_Elimination
  nodes [Output Animation Pose(Result=Sequence Player 'TerraBulb_ElimReact_PACK'() {bLoopAnimation=false, sequence=/BRCosmetics/Accessories/FORT_Backpacks/Backpack_F_MED_TerraBulb/Animation/TerraBulb_ElimReact_PACK})]
 }
 Idle {
  path=AnimGraph.AnimGraphNode_StateMachine_0.Idle and Reactive.AnimStateNode_2.Idle
  nodes [Output Animation Pose(Result=Sequence Player 'TerraBulb_BasePose_PACK'() {bLoopAnimation=false, sequence=/BRCosmetics/Accessories/FORT_Backpacks/Backpack_F_MED_TerraBulb/Animation/TerraBulb_BasePose_PACK})]
 }
 Reactive_TakeDamage {
  path=AnimGraph.AnimGraphNode_StateMachine_0.Idle and Reactive.AnimStateNode_3.Reactive_TakeDamage
  nodes [
   n1 = Sequence Player 'TerraBulb_DamageReact_PACK'() {bLoopAnimation=false, sequence=/BRCosmetics/Accessories/FORT_Backpacks/Backpack_F_MED_TerraBulb/Animation/TerraBulb_DamageReact_PACK}
   Output Animation Pose(Result=@n1)
  ]
 }
 Transition {
  path=AnimGraph.AnimGraphNode_StateMachine_0.Idle and Reactive.AnimStateTransitionNode_0.Transition
  nodes [Result(bCanEnterTransition=Get Reactive TakeDamage() -> (Reactive TakeDamage:bool))]
 }
 Transition {path=AnimGraph.AnimGraphNode_StateMachine_0.Idle and Reactive.AnimStateTransitionNode_1.Transition, nodes [Result()]}
 Transition {
  path=AnimGraph.AnimGraphNode_StateMachine_0.Idle and Reactive.AnimStateTransitionNode_2.Transition
  nodes [Result(bCanEnterTransition=Get Reactive Elimination() -> (Reactive Elimination:bool))]
 }
 Transition {path=AnimGraph.AnimGraphNode_StateMachine_0.Idle and Reactive.AnimStateTransitionNode_3.Transition, nodes [Result()]}
 FE_Reactivity {
  entry=Idle
  path=AnimGraph.AnimGraphNode_StateMachine_1.FE_Reactivity
  ids=id
  nodes [
   AnimStateNode_0 = Idle()
   AnimStateNode_2 = FE_Reactive() {is=title AnimStateNode, stateEntered {guid=#16, notifyName=Entered FE_Reactive_Anim}}
   AnimStateTransitionNode_8 = AnimStateTransitionNode() {crossfadeDuration=0}
   AnimStateTransitionNode_9 = AnimStateTransitionNode() {bAutomaticRuleBasedOnSequencePlayerInState=true, crossfadeDuration=0}
  ]
  links [
   {
    dst=Idle
    rule=AnimGraph.AnimGraphNode_StateMachine_1.FE_Reactivity.AnimStateTransitionNode_9.Transition
    src=FE_Reactive
    via=AnimStateTransitionNode_9
   }
   {
    dst=FE_Reactive
    rule=AnimGraph.AnimGraphNode_StateMachine_1.FE_Reactivity.AnimStateTransitionNode_8.Transition
    src=Idle
    via=AnimStateTransitionNode_8
   }
  ]
 }
 Idle {
  path=AnimGraph.AnimGraphNode_StateMachine_1.FE_Reactivity.AnimStateNode_0.Idle
  nodes [Output Animation Pose(Result=Sequence Player 'TerraBulb_BasePose_PACK'() {bLoopAnimation=false, sequence=&TerraBulb_BasePose_PACK})]
 }
 FE_Reactive {
  path=AnimGraph.AnimGraphNode_StateMachine_1.FE_Reactivity.AnimStateNode_2.FE_Reactive
  nodes [Output Animation Pose(Result=Sequence Player 'TerraBulb_ElimReact_PACK'() {bLoopAnimation=false, sequence=&TerraBulb_ElimReact_PACK})]
 }
 Transition {
  path=AnimGraph.AnimGraphNode_StateMachine_1.FE_Reactivity.AnimStateTransitionNode_8.Transition
  nodes [Result(bCanEnterTransition=Get Play Reactive FE Anim() -> (Play Reactive FE Anim:bool))]
 }
 Transition {path=AnimGraph.AnimGraphNode_StateMachine_1.FE_Reactivity.AnimStateTransitionNode_9.Transition, nodes [Result()]}
 EventGraph {nodes [
   Event OnCharacterCustomizationLoaded()
   > n1 = Cast To FortPlayerPawn(Object=TryGetFortPawn())
   > n2 = Branch()
   > Do Once(Reset=@n6.then)
   > n3 = Delay(Duration=3)
   Set Play Reactive FE Anim(execute=[@n3.then, @n5.then], Play Reactive FE Anim:bool=true) -> (Output_Get:bool)
   > n4 = Delay(Duration=1)
   Set Play Reactive FE Anim(execute=[@n4.then, @n7.then], Play Reactive FE Anim=false)
   > n5 = Delay(Duration=15)
   n6 = Set Play Reactive FE Anim(execute=@n2.else, Play Reactive FE Anim=false)
   n7 = AnimNotify_FE_ReactiveAnim() {is=title K2Node_Event}
   n8 = Event BlueprintProcessAnimInputEvent()
   > n9 = Sequence!(execute) -> (then_0, then_1, then_2:Exec)
   > Branch(Condition=IsNamed(self=@n8.AnimInputEvent, InEventName=aie.EnemyPlayerEliminated))
   > Set Reactive Elimination(Reactive Elimination:bool=true) -> (Output_Get:bool)
   > RetriggerableDelay(Duration=0.2)
   > Set Reactive Elimination(Reactive Elimination=false)
   Branch(execute=@n9.then_1, Condition=IsNamed(self=@n8.AnimInputEvent, InEventName=aie.ChestOpened))
   > Set Reactive OpenChest(Reactive OpenChest:bool=true) -> (Output_Get:bool)
   > RetriggerableDelay(Duration=0.2)
   > Set Reactive OpenChest(Reactive OpenChest=false)
   Branch(
    execute=@n9.then_2
    Condition=OR Boolean(
     A=IsNamed(self=@n8.AnimInputEvent, InEventName=aie.PlayerVitality.ShieldsDecreased)
     B=IsNamed(self=@n8.AnimInputEvent, InEventName=aie.PlayerVitality.HealthDecreased)
    )
   )
   > Set Reactive TakeDamage(Reactive TakeDamage:bool=true) -> (Output_Get:bool)
   > RetriggerableDelay(Duration=0.2)
   > Set Reactive TakeDamage(Reactive TakeDamage=false)
   TryGetPawnOwner()
   GetModifierForPartType!(self:Fort Player Pawn Object Reference=@n1, PartType:EFortCustomPartType Enum=Backpack) -> (ReturnValue:Custom Character Part Modifier Object Reference)
   Event BlueprintUpdateAnimation()
  ]}
]
variables [
 Play Reactive FE Anim {type=boolean}
 Play Reactive Front End Anim {type=boolean}
 Reactive Elimination {type=boolean}
 Reactive OpenChest {type=boolean}
 Reactive TakeDamage {type=boolean}
 Reactivity Enabled {type=boolean}
]
